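# Translucent UI material for a Chapter 6 Season 1 season pass tile, layering a character texture over an animated, color-themed vortex background.
# Twisted polar UVs scroll an FX texture into a tunnel and displacement that warps circle-based rim and core masks.
# A rotating, flow-mapped cloud noise forms the smoke vignette.
# An adjustable radial gradient blends over the background before the character, which has scale and offset parameters, is composited.
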Material /OfferCatalog/Art/SeasonPasses/Materials/S33/M_UI_Ch6S1_Tile
graph {
 properties {
  bAutomaticallySetUsageInEditor=false
  bUsedWithStaticMesh=false
  blendMode=BLEND_Translucent
  materialDomain=MD_UI
  usedShadingModels=Default Lit
 }
 nodes [
  Constant3Vector() {constant=<rgba 0.9873 1 0 1>}
  n1 = NamedRerouteUsage()
  n2 = LinearInterpolate(A=NamedRerouteUsage(), B=NamedRerouteUsage(), Alpha=Saturate(None=ComponentMask(None=@n1) {g=true}))
  n3 = NamedRerouteUsage()
  n4 = ScalarParameter() {defaultValue=0.25, group=Smoke, parameterName=Smoke Vignette Power}
  n5 = Saturate(None=OneMinus(None=Power(Base=@n3, Exp=@n4)))
  n6 = Add(
   A=LinearInterpolate(A=LinearInterpolate(A=@n2, Alpha=@n5) {constB=0}, B=NamedRerouteUsage(), Alpha=NamedRerouteUsage())
   B=LinearInterpolate(B=NamedRerouteUsage(), Alpha=NamedRerouteUsage()) {constB=0}
  )
  n7 = NamedRerouteUsage()
  n8 = ScalarParameter() {defaultValue=0.2, group=Radial Gradient, parameterName=Brightness}
  n9 = MaterialFunctionCall() {materialFunction=&MF_UI_UVScaling}
  n10 = ScalarParameter() {defaultValue=1, group=Character, parameterName=Texture Scale}
  n11 = ScalarParameter() {group=Character, parameterName=X Location}
  n12 = ScalarParameter() {group=Character, parameterName=Y Location}
  n13 = TextureSampleParameter2D(UVs=Add(
    A=MaterialFunctionCall(UVs=@n9.Scaled To Fill, Texture Scale=@n10) {materialFunction=&ScaleUVsByCenter}.UVs
    B=AppendVector(A=@n11, B=@n12)
   )) {
   group=Character
   parameterName=Texture
   texture=/OfferCatalog/Textures/BattleRoyale/FeaturedItems/Outfit/S28/T-AthenaSoldiers-Character-VoidRedemption-Rebel
  }
  LinearInterpolate(A=LinearInterpolate(A=Saturate(None=@n6), B=Divide(A=@n7, B=@n8) {constB=0.5}, Alpha=NamedRerouteUsage()), B=@n13.RGB, Alpha=@n13.RGB)
  MaterialFunctionCall!(Input Alpha=Constant() {r=1}, bOverrideCornerRadius, Corner Radius Override) {materialFunction=/FortUILibrary/Systems/ItemShopTile/Materials/MaterialFunctions/MF_UI_ItemShop_RoundedCorners}
  MaterialFunctionCall() {materialFunction=&MF_UI_UVScaling}
  n14 = MaterialFunctionCall(UVs=NamedRerouteUsage(), Texture Scale=ScalarParameter() {defaultValue=1, group=Radial Gradient, parameterName=Radial Gradient Scale}) {materialFunction=&ScaleUVsByCenter}
  n15 = AppendVector(A=ScalarParameter() {group=Radial Gradient, parameterName=Radial G X}, B=ScalarParameter() {group=Radial Gradient, parameterName=Radial G Y})
  n16 = MaterialFunctionCall("Value (0-1)"=MaterialFunctionCall(UV=Add(A=@n14.UVs, B=@n15), GradientStart=Constant(), GradientEnd=Constant() {r=1}) {materialFunction=&MF_UI_RadialGradient}.Smooth) {materialFunction=&MF_UI_EaseCurves}
  NamedRerouteDeclaration(None=Multiply(A=VectorParameter() {group=Radial Gradient, parameterName=Color, defaultValue=<rgba 0.7543 0 1 1>}.RGB, B=@n16.EaseInSine)) {name=Radial Gradient, nodeColor=<rgba 1 0 0.7266 1>}
  NamedRerouteDeclaration(None=Multiply(A=@n16.EaseInSine, B=ScalarParameter() {defaultValue=0.5, group=Radial Gradient, parameterName=Opacity})) {name=Radial G Mask, nodeColor=<rgba 0 1 0.9531 1>}
  NamedRerouteDeclaration(None=Constant3Vector() {constant=<rgba 0.1845 0.0044 0.0307 1>}) {name=Color Theme 2, nodeColor=<rgba 0 0 0 1>}
  NamedRerouteDeclaration(None=Constant3Vector() {constant=<rgba 0.0561 0.0027 0.0232 1>}) {name=Color Theme 3, nodeColor=<rgba 0 0 0 1>}
  NamedRerouteDeclaration(None=Constant3Vector() {constant=<rgba 0.0067 0 0.0104 1>}) {name=Color Theme 4, nodeColor=<rgba 0 0 0 1>}
  n17 = NamedRerouteUsage()
  n18 = Constant2Vector() {g=1, r=1.375}
  n19 = Saturate(None=MaterialFunctionCall(
    UV=MaterialFunctionCall(UVs=@n17, Texture Scale=@n18) {materialFunction=&ScaleUVsByCenter}.UVs
    GradientStart=Constant() {r=1}
    GradientEnd=Constant() {r=0.1}
   ) {materialFunction=&MF_UI_RadialGradient}.Smooth)
  n20 = Rotator(Coordinate=NamedRerouteUsage()) {speed=-0.05}
  n21 = TextureObject() {texture=/OfferCatalog/Art/SeasonPasses/Textures/S33/T_UI_CH6S1_CloudNoise1}
  n22 = TextureSample(
   UVs=NamedRerouteDeclaration(None=MaterialFunctionCall!(Use Local UV, Texture Size, Scale Pivot Point) {materialFunction=/Game/Environments/FrontEnd/Blueprints/ItemPreview/VaultWorldTransitions/MaterialFunctions/MF_UI_VaultWorld_UV}.UV) {name=Base UV, nodeColor=<rgba 0 0 0 1>}
   Tex=@n21
  )
  n23 = Constant() {r=0.1}
  n24 = Constant() {r=0.5}
  n25 = Multiply(A=ComponentMask(None=MaterialFunctionCall!(UVs=@n20, Texture Object=@n21, Grayscale Flow Map=@n22.RGB, Flow Direction, Flow Intensity=@n23, Flow Speed=@n24) {materialFunction=/FortUILibrary/MaterialFunctions/Effects/MF_UI_GrayscaleFlowMap_New}.Distorted Texture) {r=true}) {constB=8}
  n26 = SmoothStep(Value=Multiply(A=Saturate(None=@n25), B=@n19)) {constMax=0.75, constMin=0.1}
  NamedRerouteDeclaration(None=Saturate(None=LinearInterpolate(A=@n19, Alpha=@n26))) {name=Smoke Vignette Mask, nodeColor=<rgba 0 0 0 1>}
  n27 = NamedRerouteUsage()
  NamedRerouteDeclaration(None=Min(A=Constant() {r=0.5}, B=MaterialFunctionCall!(UV=@n27, Seed) {materialFunction=&MF_UI_Random}.Result)) {name=Noise, nodeColor=<rgba 0 0 0 1>}
  n28 = Multiply(A=NamedRerouteUsage()) {constB=0.5}
  n29 = Divide(B=Max(A=NamedRerouteUsage()) {constB=0.001}) {constA=0.125}
  NamedRerouteDeclaration(None=Power(Base=Multiply(A=@n28, B=@n29)) {constExponent=1.5}) {name=Vortex Tunnel, nodeColor=<rgba 0 0 0 1>}
  n30 = Subtract(A=MaterialFunctionCall!(UV=NamedRerouteUsage(), Twist Amount=ScalarParameter() {defaultValue=0.95, group=Vortex Distortion, parameterName=VortexTwist}) {materialFunction=/FortUILibrary/MaterialFunctions/Effects/MF_UI_Twist}.Twist UV) {constB=0.5}
  n31 = Multiply(A=Length(None=@n30) {is=class MaterialExpressionLength}) {constB=2}
  n32 = AppendVector(
   A=Multiply(A=MaterialFunctionCall!(Vector or UVs=@n30) {materialFunction=/FortUILibrary/MaterialFunctions/Basics/MF_UI_CheaperVectorToRadial}.Vector Converted to Angle) {constB=2}
   B=OneMinus(None=NamedRerouteDeclaration(None=@n31) {name=Tunnel_Mask, nodeColor=<rgba 0 0 0 1>})
  )
  n33 = AppendVector(
   A=NamedRerouteDeclaration(None=ScalarParameter() {defaultValue=8, group=Vortex Distortion, parameterName=VortexTiling}) {name=VFX_Tiling, nodeColor=<rgba 0 0 0 1>}
   B=NamedRerouteUsage()
  )
  n34 = Multiply(A=Time(), B=ScalarParameter() {defaultValue=0.125, group=Vortex Distortion, parameterName=VortexTimeFactor})
  n35 = Constant2Vector() {g=-1, r=-0.5}
  n36 = Frac(None=Multiply(A=@n34, B=ComponentMask(None=@n35) {r=true}))
  n37 = Frac(None=Multiply(A=@n34, B=ComponentMask(None=@n35) {g=true}))
  n38 = TextureSample(UVs=Add(A=Multiply(A=@n32, B=@n33), B=AppendVector(A=@n36, B=@n37))) {samplerType=SAMPLERTYPE_Alpha, texture=/OfferCatalog/Art/SeasonPasses/Textures/S33/T_UI_BP_FX}
  NamedRerouteDeclaration(None=Multiply(A=SquareRoot(None=@n38.G), B=ScalarParameter() {defaultValue=0.125, group=Vortex Distortion, parameterName=VortexDistortionAmount})) {name=Displacement, nodeColor=<rgba 0 0 0 1>}
  n39 = NamedRerouteUsage()
  n40 = ScalarParameter() {defaultValue=0.35, group=Vortex Rim, parameterName=RimDistortionAmount}
  n41 = MaterialFunctionCall(
   UVs=Add(A=Multiply(A=@n39, B=@n40), B=NamedRerouteUsage())
   Size=ScalarParameter() {defaultValue=0.3, group=Vortex Rim, parameterName=RimSize}
   OutlineThickness=ScalarParameter() {defaultValue=0.002, group=Vortex Rim, parameterName=RimOutline}
   GlowMax=ScalarParameter() {defaultValue=0.075, group=Vortex Rim, parameterName=RimGlow}
  ) {materialFunction=&MF_UI_Circle}
  n42 = Multiply(A=@n32, B=ScalarParameter() {defaultValue=1.5, group=Vortex Rim, parameterName=RimTunnelMaskIntensity})
  NamedRerouteDeclaration(None=MaterialFunctionCall("Value (0-1)"=Multiply(A=@n41.Fill, B=@n42)) {materialFunction=&MF_UI_EaseCurves}.EaseOutSine) {name=Vortex Rim, nodeColor=<rgba 0 0 0 1>}
  n43 = Multiply(A=NamedRerouteUsage(), B=NamedRerouteUsage())
  n44 = NamedRerouteUsage()
  n45 = Power(
   Base=MaterialFunctionCall(
    UVs=Add(A=@n43, B=@n44)
    Size=ScalarParameter() {defaultValue=0.3, group=Vortex Core, parameterName=Vortex Core Size}
    GlowMax=ScalarParameter() {defaultValue=0.75, group=Vortex Core, parameterName=Vortext Core Glow}
   ) {materialFunction=&MF_UI_Circle}.Fill
   Exp=ScalarParameter() {defaultValue=12, group=Vortex Core, parameterName=Vortx Core Power}
  ) {constExponent=5}
  n46 = ScalarParameter() {defaultValue=0.5, group=Vortex Core, parameterName=Vortext Core Intensity}
  NamedRerouteDeclaration(None=Saturate(None=Multiply(A=@n45, B=@n46) {constB=0.125})) {name=Vortext Core, nodeColor=<rgba 0 0 0 1>}
  NamedRerouteDeclaration(None=MaterialFunctionCall(
    UVs=MaterialFunctionCall!(Vault World Color, Focal Color Override, Focal Scale Override, Mask Blend) {materialFunction=/Game/Environments/FrontEnd/Blueprints/ItemPreview/VaultWorldTransitions/MaterialFunctions/MF_UI_VaultWorld_FocalPoint}.Focal Point UV
    Texture Scale=ScalarParameter() {defaultValue=1.8, group=Vortex Core, parameterName=Size of Vertex}
   ) {materialFunction=&ScaleUVsByCenter}.UVs) {name=UV, nodeColor=<rgba 0 0 0 1>}
  NamedRerouteDeclaration(None=Constant3Vector() {constant=<rgba 1 0 0.1441 1>}) {name=Color Theme 1, nodeColor=<rgba 0 0 0 1>}
  NamedRerouteUsage()
  OneMinus(None=@n31)
 ]
}
