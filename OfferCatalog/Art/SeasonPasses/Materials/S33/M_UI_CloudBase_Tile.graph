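# Translucent UI material for a Season 33 pass tile that builds an animated cloud backdrop from color parameters, gradients and optional background noise.
# Background and foreground cloud layers use panned flow-map distortion and dual linear-gradient masks, then combine into cloud color and highlights.
# A rotated radial beam mixes cloud contrast with time-driven rays from polar-remapped noise.
# Further branches provide an optional vignette, a radial glow, a character texture and item-shop rounded corners.
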Material /OfferCatalog/Art/SeasonPasses/Materials/S33/M_UI_CloudBase_Tile
graph {
 properties {
  bAutomaticallySetUsageInEditor=false
  bUsedWithStaticMesh=false
  blendMode=BLEND_Translucent
  materialDomain=MD_UI
  usedShadingModels=Default Lit
 }
 nodes [
  n1 = MaterialFunctionCall() {materialFunction=&MF_UI_UVScaling}
  n2 = ScalarParameter() {defaultValue=1, group=Character, parameterName=Texture Scale}
  n3 = ScalarParameter() {group=Character, parameterName=X Location}
  n4 = ScalarParameter() {group=Character, parameterName=Y Location}
  n5 = TextureSampleParameter2D(UVs=Add(
    A=MaterialFunctionCall(UVs=@n1.Scaled To Fill, Texture Scale=@n2) {materialFunction=&ScaleUVsByCenter}.UVs
    B=AppendVector(A=@n3, B=@n4)
   )) {
   group=Character
   parameterName=Texture
   texture=/OfferCatalog/Textures/BattleRoyale/FeaturedItems/Outfit/S28/T-AthenaSoldiers-Character-VoidRedemption-Rebel
  }
  LinearInterpolate(A=NamedRerouteUsage(), B=@n5.RGB, Alpha=@n5.RGB)
  n6 = MaterialFunctionCall(UVs=NamedRerouteUsage(), Texture Scale=Constant() {r=3}) {materialFunction=&ScaleUVsByCenter}
  n7 = AppendVector(A=Constant() {r=-0.1}, B=Constant() {r=0.1})
  n8 = MaterialFunctionCall("Value (0-1)"=MaterialFunctionCall(UV=Add(A=@n6.UVs, B=@n7), GradientStart=Constant(), GradientEnd=Constant() {r=1}) {materialFunction=&MF_UI_RadialGradient}.Smooth) {materialFunction=&MF_UI_EaseCurves}
  LinearInterpolate(A=LinearInterpolate(
    A=Constant3Vector() {constant=<rgba 0.0116 0.1022 0.0999 1>}
    B=Constant3Vector() {constant=<rgba 0.0887 0.1912 0.1651 1>}
    Alpha=Multiply(A=@n8.EaseInSine) {constB=2}
   ))
  MaterialFunctionCall!(Input Alpha=Constant() {r=1}, bOverrideCornerRadius, Corner Radius Override) {materialFunction=/FortUILibrary/Systems/ItemShopTile/Materials/MaterialFunctions/MF_UI_ItemShop_RoundedCorners}
  n9 = MaterialFunctionCall(UVs=NamedRerouteUsage(), Texture Scale=ScalarParameter() {defaultValue=1, group=Radial Gradient, parameterName=Radial Gradient Scale}) {materialFunction=&ScaleUVsByCenter}
  n10 = AppendVector(
   A=ScalarParameter() {group=Radial Gradient, parameterName=Radial G X}
   B=ScalarParameter() {defaultValue=1, group=Radial Gradient, parameterName=Radial G Y}
  )
  n11 = MaterialFunctionCall("Value (0-1)"=MaterialFunctionCall(UV=Add(A=@n9.UVs, B=@n10), GradientStart=Constant(), GradientEnd=Constant() {r=1}) {materialFunction=&MF_UI_RadialGradient}.Smooth) {materialFunction=&MF_UI_EaseCurves}
  NamedRerouteDeclaration(None=Multiply(A=VectorParameter() {group=Radial Gradient, parameterName=Color, defaultValue=<rgba 0.0185 0.0185 0.0185 1>}.RGB, B=@n11.EaseInSine)) {name=Radial Gradient, nodeColor=<rgba 1 0 0.7266 1>}
  NamedRerouteDeclaration(None=Multiply(A=@n11.EaseInSine, B=ScalarParameter() {defaultValue=0.5, group=Radial Gradient, parameterName=Opacity})) {name=Radial G Mask, nodeColor=<rgba 0 1 0.9531 1>}
  n12 = NamedRerouteUsage()
  n13 = TextureObjectParameter() {
   group=02 - Background Texture
   parameterName=Background Texture
   samplerType=SAMPLERTYPE_Masks
   texture=/Game/Global/Textures/Default/Noise/T_Metal_Noise
  }
  n14 = Multiply(
   A=NamedRerouteUsage()
   B=LinearInterpolate(
    A=ScalarParameter() {group=02 - Background Texture, parameterName=Background Texture Min Value, sliderMax=1}
    Alpha=TextureSample(UVs=@n12, Tex=@n13) {samplerType=SAMPLERTYPE_Masks, texture=&T_Metal_Noise}.R
   )
  )
  n15 = NamedRerouteUsage()
  n16 = ScalarParameter() {defaultValue=0.5, group=02 - Background Texture, parameterName=Background Noise Tile X}
  n17 = ScalarParameter() {defaultValue=0.5, group=02 - Background Texture, parameterName=Background Noise Tile Y}
  n18 = TextureObjectParameter() {
   group=02 - Background Texture
   parameterName=Background Noise Texture
   samplerType=SAMPLERTYPE_Masks
   texture=/Game/Global/Textures/Default/Noise/T_Pattern_Noise_M
  }
  n19 = TextureSample(UVs=Multiply(A=@n12, B=AppendVector(A=@n16, B=@n17)), Tex=@n18) {samplerType=SAMPLERTYPE_Masks, texture=&T_Pattern_Noise_M}
  NamedRerouteDeclaration(None=StaticSwitchParameter(True=LinearInterpolate(A=@n14, B=@n15, Alpha=@n19.R), False=NamedRerouteUsage()) {group=02 - Background Texture, parameterName=Use Background Texture, sortPriority=-32}) {name=Colored Background Texture, nodeColor=<rgba 0 0 0 1>}
  n20 = MaterialFunctionCall() {materialFunction=&GetUserInterfaceUV}
  NamedRerouteDeclaration(None=@n20.Normalized UV) {name=Normalized UV, nodeColor=<rgba 0 0 0 1>}
  NamedRerouteDeclaration(None=MaterialFunctionCall() {materialFunction=&MF_UI_UVScaling}.Scaled to Fit) {name=UV, nodeColor=<rgba 1 0.8438 0 1>}
  n21 = NamedRerouteUsage()
  n22 = NamedRerouteUsage()
  n23 = NamedRerouteUsage()
  n24 = ScalarParameter() {group=01 - Gradient, parameterName=Gradient Rotation}
  n25 = NamedRerouteUsage()
  n26 = ScalarParameter() {defaultValue=1, group=01 - Gradient, parameterName=Radial Gradient Scale}
  n27 = ScalarParameter() {group=01 - Gradient, parameterName=Radial Gradient Offset X}
  n28 = ScalarParameter() {group=01 - Gradient, parameterName=Radial Gradient Offset Y}
  n29 = MaterialFunctionCall(UV=Add(
    A=MaterialFunctionCall(UVs=@n25, Texture Scale=@n26) {materialFunction=&ScaleUVsByCenter}.UVs
    B=AppendVector(A=@n27, B=@n28)
   ) {constB=0.5}) {materialFunction=&MF_UI_UVScaling}
  n30 = Saturate(None=StaticSwitchParameter(
    True=MaterialFunctionCall(UVs=@n23, RotationAngle=@n24) {materialFunction=&MF_UI_LinearGradient}.Result
    False=MaterialFunctionCall(UV=@n29.Scaled To Fill) {materialFunction=&MF_UI_RadialGradient}.Smooth
   ) {group=01 - Gradient, parameterName=Use Linear Gradient, sortPriority=-32})
  NamedRerouteDeclaration(None=StaticSwitchParameter(True=LinearInterpolate(A=@n21, B=@n22, Alpha=@n30), False=NamedRerouteUsage()) {defaultValue=true, group=01 - Gradient, parameterName=Use Gradient, sortPriority=-32}) {name=Colored Gradient, nodeColor=<rgba 0 0 0 1>}
  n31 = ScalarParameter() {group=0 - Global, parameterName=Clouds Highlight Sharpening}
  n32 = Multiply(A=NamedRerouteUsage(), B=NamedRerouteUsage())
  n33 = NamedRerouteUsage()
  n34 = NamedRerouteUsage()
  n35 = NamedRerouteDeclaration(None=StaticSwitchParameter(True=LinearInterpolate(A=@n32, B=@n33, Alpha=@n34), False=Multiply(A=@n33, B=@n34)) {defaultValue=true, group=03 - Background Clouds, parameterName=Use Background Clouds, sortPriority=-32}) {name=Combined Grayscale Clouds, nodeColor=<rgba 0 0 0 1>}
  n36 = LinearInterpolate(
   A=NamedRerouteUsage()
   Alpha=Multiply(A=SmoothStep(Min=@n31, Value=@n35) {constMin=0.8}, B=ScalarParameter() {group=0 - Global, parameterName=Clouds Highlight})
  )
  NamedRerouteDeclaration(None=MaterialFunctionCall(
    A=Constant()
    B=Saturate(None=@n36)
    C=Constant() {r=1}
    Alpha=ScalarParameter() {defaultValue=0.5, group=0 - Global, parameterName=Luminosity}
   ) {materialFunction=&Lerp_3Color}.Result) {name=Clouds Highlights, nodeColor=<rgba 0 0 0 1>}
  n37 = Rotator(
   Coordinate=MaterialFunctionCall(UV=NamedRerouteUsage(), ScalePivotPoint=Constant2Vector() {g=0.5, r=1}) {materialFunction=&MF_UI_UVScaling}.ScaleX
   Time=ScalarParameter() {defaultValue=0.125, group=06 - Light Beam, parameterName=Light Beam Rotation}
  ) {speed=6.2832}
  n38 = Constant2Vector() {g=0.5, r=0.125}
  n39 = Add(A=Multiply(A=Subtract(A=@n37, B=@n38) {constB=0.5}, B=Constant2Vector() {g=1, r=0.25}), B=@n38) {constB=0.5}
  n40 = NamedRerouteDeclaration(None=MaterialFunctionCall(UV=NamedRerouteDeclaration(None=@n39) {name=Light Beam UV, nodeColor=<rgba 0 0 0 1>}) {materialFunction=&MF_UI_RadialGradient}.Smooth) {name=Light Beam Mask, nodeColor=<rgba 0 0 0 1>}
  n41 = ScalarParameter() {defaultValue=1, group=06 - Light Beam, parameterName=Mask Rays, sliderMax=1}
  n42 = OneMinus(None=NamedRerouteDeclaration(None=ScalarParameter() {defaultValue=1, group=06 - Light Beam, parameterName=Beam Cloud Opacity, sliderMax=1}) {name=Beam Cloud Opacity, nodeColor=<rgba 0 0 0 1>})
  n43 = Power(Base=NamedRerouteUsage(), Exp=ScalarParameter() {defaultValue=0.25, group=06 - Light Beam, parameterName=Beam Cloud Contrast}) {constExponent=0.25}
  n44 = NamedRerouteDeclaration(None=Multiply(A=LinearInterpolate(B=@n40, Alpha=@n41) {constA=1}, B=LinearInterpolate(A=@n42, Alpha=@n43))) {name=Clouds Beam, nodeColor=<rgba 0 0 0 1>}
  n45 = ScalarParameter() {defaultValue=1, group=06 - Light Beam, parameterName=Rays Opacity, sliderMax=1}
  n46 = MaterialFunctionCall(UV=NamedRerouteUsage(), ScalePivotPoint=Constant2Vector() {r=1}) {materialFunction=&MF_UI_UVScaling}
  n47 = Constant2Vector() {g=1, r=-1}
  n48 = ScalarParameter() {defaultValue=0.1, group=06 - Light Beam, parameterName=Light Rays Speed}
  n49 = Time()
  n50 = MaterialFunctionCall!(Vector or UVs=Rotator(Coordinate=Add(A=@n46.Scaled To Fill, B=@n47), Time=Multiply(A=@n48, B=@n49)) {speed=6.2832}) {materialFunction=/FortUILibrary/MaterialFunctions/Basics/MF_UI_CheaperVectorToRadial}
  n51 = NamedRerouteUsage()
  n52 = NamedRerouteDeclaration(None=TextureObject() {samplerType=SAMPLERTYPE_Grayscale, texture=&LowResBlurredNoise}) {name=Rays Texture, nodeColor=<rgba 0 0 0 1>}
  n53 = TextureSample(UVs=Multiply(A=@n50.Vector Converted to Angle, B=Constant2Vector() {g=-1, r=1}) {constB=2}, Tex=@n51)
  n54 = Min(
   A=TextureSample(UVs=TextureSample(UVs=@n50.Vector Converted to Angle, Tex=@n51).R, Tex=@n52).R
   B=TextureSample(UVs=Multiply(A=@n53.R) {constB=1.5}, Tex=@n52).R
  )
  n55 = Saturate(None=Power(
    Base=NamedRerouteDeclaration(None=@n54) {name=Light Rays, nodeColor=<rgba 0 0 0 1>}
    Exp=ScalarParameter() {defaultValue=1.5, group=06 - Light Beam, parameterName=Light Rays Contrast}
   ))
  n56 = LinearInterpolate(A=@n44, B=NamedRerouteDeclaration(None=Multiply(A=@n45, B=@n55)) {name=Rays Opacity, nodeColor=<rgba 0 0 0 1>}, Alpha=@n44)
  n57 = NamedRerouteUsage()
  n58 = Saturate(None=Multiply(A=@n56) {constB=4})
  n59 = Multiply(
   A=NamedRerouteDeclaration(None=@n56) {name=Combined Rays and Clouds Beam, nodeColor=<rgba 0 0 0 1>}
   B=Saturate(None=MaterialFunctionCall!(Base=@n57, Add=@n58) {materialFunction=/Engine/Functions/Engine_MaterialFunctions02/Math/MultiplyAdd}.Result)
  )
  NamedRerouteDeclaration(None=Saturate(None=Add(A=@n59, B=@n59))) {name=Colored Light Beam, nodeColor=<rgba 0 0 0 1>}
  n60 = NamedRerouteUsage()
  n61 = Multiply(A=@n60, B=OneMinus(None=ScalarParameter() {defaultValue=1, group=05 - Vignette, parameterName=Vignette Intensity, sliderMax=1})) {constB=0.1}
  n62 = NamedRerouteUsage()
  n63 = ScalarParameter() {defaultValue=1.25, group=05 - Vignette, parameterName=Vignette Scale}
  n64 = OneMinus(None=MaterialFunctionCall(UV=MaterialFunctionCall(UVs=@n62, Texture Scale=@n63) {materialFunction=&ScaleUVsByCenter}.UVs) {materialFunction=&MF_UI_RadialGradient}.Smooth)
  NamedRerouteDeclaration(None=StaticSwitchParameter(True=LinearInterpolate(A=@n60, B=@n61, Alpha=@n64) {constB=0}, False=NamedRerouteUsage()) {group=05 - Vignette, parameterName=Display Vignette}) {name=Vignette Background, nodeColor=<rgba 0 0 0 1>}
  n65 = Constant() {r=1}
  n66 = Add(
   A=NamedRerouteUsage()
   B=AppendVector(A=Constant(), B=ScalarParameter() {group=04 - Foreground Clouds, parameterName=Foreground Cloud Masks Offset Y})
  )
  n67 = ScalarParameter() {group=04 - Foreground Clouds, parameterName=Foreground Cloud Mask Start}
  n68 = ScalarParameter() {defaultValue=0.5, group=04 - Foreground Clouds, parameterName=Foreground Cloud Mask End}
  n69 = Add(A=ScalarParameter() {defaultValue=0.125, group=04 - Foreground Clouds, parameterName=Foreground Cloud Mask Rotation}) {constB=0.5}
  n70 = ScalarParameter() {defaultValue=0.5, group=04 - Foreground Clouds, parameterName=Foreground Cloud Mask Offset}
  n71 = ScalarParameter() {group=04 - Foreground Clouds, parameterName=Foreground Cloud Mask 2 Start}
  n72 = ScalarParameter() {defaultValue=0.5, group=04 - Foreground Clouds, parameterName=Foreground Cloud Mask 2 End}
  n73 = OneMinus(None=ScalarParameter() {defaultValue=0.125, group=04 - Foreground Clouds, parameterName=Foreground Cloud Mask 2 Rotation})
  n74 = ScalarParameter() {defaultValue=0.5, group=04 - Foreground Clouds, parameterName=Foreground Cloud Mask 2 Offset}
  n75 = Max(
   A=Saturate(None=MaterialFunctionCall(UVs=@n66, GradientStart=@n67, GradientEnd=@n68, RotationAngle=@n69, GradientOffset=@n70) {materialFunction=&MF_UI_LinearGradient}.Result)
   B=Saturate(None=MaterialFunctionCall(UVs=@n66, GradientStart=@n71, GradientEnd=@n72, RotationAngle=@n73, GradientOffset=@n74) {materialFunction=&MF_UI_LinearGradient}.Result)
  )
  n76 = Multiply(A=@n75, B=NamedRerouteUsage())
  n77 = LinearInterpolate(
   A=@n75
   B=Saturate(None=Multiply(A=@n76, B=@n76))
   Alpha=ScalarParameter() {group=04 - Foreground Clouds, parameterName=Foreground Cloud Mask Blend, sliderMax=1}
  )
  n78 = ScalarParameter() {defaultValue=1, group=04 - Foreground Clouds, parameterName=Use Foreground Cloud Mask, sliderMax=1}
  NamedRerouteDeclaration(None=Multiply(
    A=ScalarParameter() {defaultValue=1, group=04 - Foreground Clouds, parameterName=Foreground Clouds Opacity, sliderMax=1}
    B=LinearInterpolate(A=@n65, B=@n77, Alpha=@n78)
   )) {name=Foreground Cloud Mask, nodeColor=<rgba 0 0 0 1>}
  n79 = ScalarParameter() {group=03 - Background Clouds, parameterName=Background Clouds Min}
  n80 = ScalarParameter() {defaultValue=1, group=03 - Background Clouds, parameterName=Background Clouds Max}
  n81 = Multiply(A=NamedRerouteUsage(), B=ScalarParameter() {defaultValue=1.05, group=03 - Background Clouds, parameterName=Background Clouds Scale}) {constB=1.05}
  n82 = TextureObjectParameter() {
   group=03 - Background Clouds
   parameterName=Background Cloud Texture
   samplerType=SAMPLERTYPE_Alpha
   texture=/Game/UI/Frontend/ItemShop/Art/LastVoice/T_UI_CloudLV
  }
  n83 = Panner(Coordinate=@n81) {speedX=0.001, speedY=0.001}
  n84 = TextureObjectParameter() {
   group=03 - Background Clouds
   parameterName=Background Cloud Flowmap
   samplerType=SAMPLERTYPE_Grayscale
   texture=/FortUILibrary/MaterialFunctions/ExampleContent/T_Cloud_R
  }
  n85 = ComponentMask(None=MaterialFunctionCall!(
    UVs=Panner(Coordinate=@n81) {speedX=0.01, speedY=0.001}
    Texture Object=@n82
    Grayscale Flow Map=TextureSample(UVs=@n83, Tex=@n84).RGB
    Flow Direction=ScalarParameter() {defaultValue=0.5, group=03 - Background Clouds, parameterName=Background Clouds Flow Direction}
    Flow Intensity=ScalarParameter() {defaultValue=0.07, group=03 - Background Clouds, parameterName=Background Clouds Flow Intensity}
    Flow Speed=ScalarParameter() {defaultValue=0.25, group=03 - Background Clouds, parameterName=Background Clouds Flow Speed}
   ) {materialFunction=/FortUILibrary/MaterialFunctions/Effects/MF_UI_GrayscaleFlowMap_New}.Distorted Texture) {r=true}
  NamedRerouteDeclaration(None=Saturate(None=SmoothStep(Min=@n79, Max=@n80, Value=@n85))) {name=Background Clouds Grayscale, nodeColor=<rgba 0 0 0 1>}
  NamedRerouteDeclaration(None=NamedRerouteUsage()) {name=Final Pass, nodeColor=<rgba 0 0 0 1>}
  n86 = ScalarParameter() {group=04 - Foreground Clouds, parameterName=Foreground Clouds Min}
  n87 = ScalarParameter() {defaultValue=1, group=04 - Foreground Clouds, parameterName=Foreground Clouds Max}
  n88 = Multiply(A=NamedRerouteUsage(), B=ScalarParameter() {defaultValue=0.6, group=04 - Foreground Clouds, parameterName=Foreground Clouds Scale}) {constB=0.65}
  n89 = TextureObjectParameter() {
   group=04 - Foreground Clouds
   parameterName=Foreground Cloud Texture
   samplerType=SAMPLERTYPE_Alpha
   texture=&T_UI_CloudLV
  }
  n90 = Panner(
   Coordinate=@n88
   Speed=AppendVector(
    A=ScalarParameter() {defaultValue=0.001, group=04 - Foreground Clouds, parameterName=Foreground Cloud Speed X}
    B=ScalarParameter() {defaultValue=0.01, group=04 - Foreground Clouds, parameterName=Foreground Cloud Speed Y}
   )
  ) {speedX=0.01, speedY=0.01}
  n91 = TextureObjectParameter() {
   group=04 - Foreground Clouds
   parameterName=Foreground Cloud Flowmap
   samplerType=SAMPLERTYPE_Grayscale
   texture=&T_Cloud_R
  }
  n92 = ComponentMask(None=MaterialFunctionCall(
    UVs=Panner(Coordinate=@n88) {speedX=0.001, speedY=0.01}
    Texture Object=@n89
    Grayscale Flow Map=TextureSample(UVs=@n90, Tex=@n91).RGB
    Flow Direction=ScalarParameter() {defaultValue=0.5, group=04 - Foreground Clouds, parameterName=Foreground Clouds Flow Direction}
    Flow Intensity=ScalarParameter() {defaultValue=0.01, group=04 - Foreground Clouds, parameterName=Foreground Clouds Flow Intensity}
    Flow Speed=ScalarParameter() {defaultValue=0.25, group=04 - Foreground Clouds, parameterName=Foreground Clouds Flow Speed}
   ) {materialFunction=&MF_UI_GrayscaleFlowMap_New}.Distorted Texture) {r=true}
  NamedRerouteDeclaration(None=Saturate(None=SmoothStep(Min=@n86, Max=@n87, Value=@n92))) {name=Foreground Clouds Grayscale, nodeColor=<rgba 0 0 0 1>}
  n93 = Constant() {r=1}
  n94 = Add(
   A=NamedRerouteUsage()
   B=AppendVector(A=Constant(), B=ScalarParameter() {group=03 - Background Clouds, parameterName=Background Cloud Masks Offset Y})
  )
  n95 = ScalarParameter() {group=03 - Background Clouds, parameterName=Background Cloud Mask Start}
  n96 = ScalarParameter() {defaultValue=0.5, group=03 - Background Clouds, parameterName=Background Cloud Mask End}
  n97 = Add(A=ScalarParameter() {defaultValue=0.125, group=03 - Background Clouds, parameterName=Background Cloud Mask Rotation}) {constB=0.5}
  n98 = ScalarParameter() {defaultValue=0.5, group=03 - Background Clouds, parameterName=Background Cloud Mask Offset}
  n99 = ScalarParameter() {group=03 - Background Clouds, parameterName=Background Cloud Mask 2 Start}
  n100 = ScalarParameter() {defaultValue=0.5, group=03 - Background Clouds, parameterName=Background Cloud Mask 2 End}
  n101 = OneMinus(None=ScalarParameter() {defaultValue=0.125, group=03 - Background Clouds, parameterName=Background Cloud Mask 2 Rotation})
  n102 = ScalarParameter() {defaultValue=0.5, group=03 - Background Clouds, parameterName=Background Cloud Mask 2 Offset}
  n103 = Max(
   A=Saturate(None=MaterialFunctionCall(UVs=@n94, GradientStart=@n95, GradientEnd=@n96, RotationAngle=@n97, GradientOffset=@n98) {materialFunction=&MF_UI_LinearGradient}.Result)
   B=Saturate(None=MaterialFunctionCall(UVs=@n94, GradientStart=@n99, GradientEnd=@n100, RotationAngle=@n101, GradientOffset=@n102) {materialFunction=&MF_UI_LinearGradient}.Result)
  )
  n104 = Multiply(A=@n103, B=NamedRerouteUsage())
  n105 = LinearInterpolate(
   A=@n103
   B=Saturate(None=Multiply(A=@n104, B=@n104))
   Alpha=ScalarParameter() {group=03 - Background Clouds, parameterName=Background Cloud Mask Blend, sliderMax=1}
  )
  n106 = ScalarParameter() {group=03 - Background Clouds, parameterName=Use Background Cloud Mask, sliderMax=1}
  NamedRerouteDeclaration(None=Multiply(
    A=ScalarParameter() {defaultValue=1, group=03 - Background Clouds, parameterName=Background Clouds Opacity, sliderMax=1}
    B=LinearInterpolate(A=@n93, B=@n105, Alpha=@n106)
   )) {name=Background Cloud Mask, nodeColor=<rgba 0 0 0 1>}
  NamedRerouteDeclaration(None=StaticSwitchParameter(
    True=MaterialFunctionCall!(ScreenUV, TextureSize, ScalePivotPoint) {materialFunction=/FortUILibrary/MaterialFunctions/Basics/MF_UI_ScreenSpaceUVScaling}.Scaled to Fit
    False=NamedRerouteUsage()
   ) {group=0 - Global, parameterName=Use ScreenSpace UV for Clouds}) {name=Clouds UV, nodeColor=<rgba 0 0 0 1>}
  NamedRerouteDeclaration(None=VectorParameter() {group=Colors, parameterName=Accent Color, defaultValue=<rgba 0 0 0 1>}.RGB) {name=Accent Color, nodeColor=<rgba 0 0 0 1>}
  n107 = NamedRerouteUsage()
  n108 = NamedRerouteUsage()
  n109 = LinearInterpolate(A=NamedRerouteUsage(), B=NamedRerouteUsage(), Alpha=NamedRerouteUsage())
  n110 = ScalarParameter() {group=01 - Gradient, parameterName=Use Cloud Gradient for Foreground, sliderMax=1}
  n111 = ScalarParameter() {defaultValue=2, group=04 - Foreground Clouds, parameterName=Foreground Clouds Color Intensity}
  n112 = NamedRerouteUsage()
  n113 = Saturate(None=MaterialFunctionCall(Base=LinearInterpolate(A=@n108, B=@n109, Alpha=@n110), Add=Multiply(A=@n111, B=@n112) {constB=2}) {materialFunction=&MultiplyAdd}.Result)
  n114 = NamedRerouteUsage()
  n115 = NamedRerouteUsage()
  n116 = LinearInterpolate(
   A=NamedRerouteUsage()
   B=LinearInterpolate(A=NamedRerouteUsage(), B=NamedRerouteUsage(), Alpha=@n115)
   Alpha=ScalarParameter() {group=01 - Gradient, parameterName=Use Cloud Gradient for Background, sliderMax=1}
  )
  n117 = Multiply(A=@n115, B=ScalarParameter() {defaultValue=2, group=03 - Background Clouds, parameterName=Background Clouds Color Intensity}) {constB=2}
  n118 = LinearInterpolate(
   A=NamedRerouteUsage()
   B=Saturate(None=MaterialFunctionCall(Base=@n116, Add=@n117) {materialFunction=&MultiplyAdd}.Result)
   Alpha=NamedRerouteUsage()
  )
  n119 = NamedRerouteUsage()
  NamedRerouteDeclaration(None=LinearInterpolate(
    A=StaticSwitchParameter(True=@n118, False=@n119) {defaultValue=true, group=03 - Background Clouds, parameterName=Use Background Clouds, sortPriority=-32}
    B=LinearInterpolate(A=@n107, B=@n113, Alpha=@n114)
    Alpha=NamedRerouteUsage()
   )) {name=Clouds Color, nodeColor=<rgba 0 0 0 1>}
  n120 = MaterialFunctionCall() {materialFunction=&MF_UI_UVScaling}
  NamedRerouteDeclaration(None=@n120.ScaleX) {name=Scale X UV, nodeColor=<rgba 0 0 0 1>}
  NamedRerouteDeclaration(None=@n20.9-Slice UV) {name=9 Slice UV, nodeColor=<rgba 0 0 0 1>}
  NamedRerouteDeclaration(None=VectorParameter() {group=Colors, parameterName=Primary Color, defaultValue=<rgba 0 0 0 1>}.RGB) {name=Primary Color, nodeColor=<rgba 0 0 0 1>}
  NamedRerouteDeclaration(None=VectorParameter() {group=Colors, parameterName=Secondary Color, defaultValue=<rgba 0 0 0 1>}.RGB) {name=Secondary Color, nodeColor=<rgba 0 0 0 1>}
  NamedRerouteDeclaration(None=@n120.ScaleY) {name=Scale Y UV, nodeColor=<rgba 0 0 0 1>}
  NamedRerouteDeclaration(None=@n120.Scaled To Fill) {name=Scale to Fill, nodeColor=<rgba 0 0 0 1>}
 ]
}
