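# Translucent UI material for a Season 33 "Figment" tile, with a character texture over a stylized animated shape background.
# Triangle masks and a gradient-faded four-sided midshape, each with placement parameters, sway through rotators with small time-driven offsets.
# Color parameters, an eased radial-gradient backdrop and the shape masks feed layered lerps beneath the character.
# The character texture's own RGB serves as its blend alpha, and an item shop rounded-corners function is included.
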Material /OfferCatalog/Art/SeasonPasses/Materials/S33/M_UI_Figment_Tile
graph {
 properties {
  bAutomaticallySetUsageInEditor=false
  bUsedWithStaticMesh=false
  blendMode=BLEND_Translucent
  materialDomain=MD_UI
  usedShadingModels=Default Lit
 }
 nodes [
  n1 = Add(
   A=NamedRerouteUsage()
   B=AppendVector(
    A=ScalarParameter() {group=Secondary Shape, parameterName=S X Location}
    B=ScalarParameter() {group=Secondary Shape, parameterName=S Y Location}
   )
  )
  n2 = ScalarParameter() {group=Secondary Shape, parameterName=S Rotation}
  n3 = ScalarParameter() {defaultValue=1, group=Secondary Shape, parameterName=S Scale X}
  n4 = ScalarParameter() {defaultValue=1, group=Secondary Shape, parameterName=S Scale Y}
  n5 = NamedRerouteUsage()
  n6 = Rotator(
   Coordinate=MaterialFunctionCall(
    UVs=MaterialFunctionCall(UVs=@n1, "Rotation Angle (0-1)"=@n2) {materialFunction=&CustomRotator}.Rotated Values
    Texture Scale=AppendVector(A=@n3, B=@n4)
   ) {materialFunction=&ScaleUVsByCenter}.UVs
   Time=Add(
    A=LinearInterpolate(Alpha=@n5) {constA=-0.01, constB=0.01}
    B=ScalarParameter() {defaultValue=0.085, group=Shapes, parameterName=Shape 03 Rotation}
   )
  ) {speed=6.2832}
  n7 = Constant2Vector() {g=0.5, r=1}
  n8 = Saturate(None=ComponentMask(None=@n6) {r=true})
  Add(A=Multiply(A=Subtract(A=@n6, B=@n7), B=LinearInterpolate(Alpha=@n8) {constA=0.5, constB=0.5}), B=@n7)
  n9 = NamedRerouteUsage()
  n10 = Constant() {r=3}
  n11 = Constant() {r=-0.1}
  n12 = Constant() {r=0.1}
  n13 = MaterialFunctionCall(
   UV=Add(
    A=MaterialFunctionCall(UVs=@n9, Texture Scale=@n10) {materialFunction=&ScaleUVsByCenter}.UVs
    B=AppendVector(A=@n11, B=@n12)
   )
   GradientStart=Constant()
   GradientEnd=Constant() {r=1}
  ) {materialFunction=&MF_UI_RadialGradient}
  n14 = LinearInterpolate(
   A=VectorParameter() {group=Color, parameterName=BackgroundColor_A, defaultValue=<rgba 0.0561 0.227 0.5841 1>}.RGB
   B=VectorParameter() {group=Color, parameterName=BackgroundColor_B, defaultValue=<rgba 0.2307 0.5457 0.7835 1>}.RGB
   Alpha=Multiply(A=MaterialFunctionCall("Value (0-1)"=@n13.Smooth) {materialFunction=&MF_UI_EaseCurves}.EaseInSine) {constB=2}
  )
  n15 = Divide(A=NamedRerouteUsage()) {constB=0.8}
  n16 = NamedRerouteUsage()
  n17 = LinearInterpolate(
   A=LinearInterpolate(
    A=LinearInterpolate(A=@n14, B=@n15, Alpha=@n16)
    B=VectorParameter() {group=Color, parameterName=LeftTriangle_Color, defaultValue=<rgba 0 0.6878 1 1>}.RGB
    Alpha=NamedRerouteUsage()
   ) {constA=1}
   B=NamedRerouteUsage()
   Alpha=NamedRerouteUsage()
  ) {constA=1}
  n18 = NamedRerouteUsage()
  n19 = NamedRerouteUsage()
  n20 = MaterialFunctionCall() {materialFunction=&MF_UI_UVScaling}
  n21 = ScalarParameter() {defaultValue=1, group=Character, parameterName=Texture Scale}
  n22 = ScalarParameter() {group=Character, parameterName=X Location}
  n23 = ScalarParameter() {group=Character, parameterName=Y Location}
  n24 = TextureSampleParameter2D(UVs=Add(
    A=MaterialFunctionCall(UVs=@n20.Scaled To Fill, Texture Scale=@n21) {materialFunction=&ScaleUVsByCenter}.UVs
    B=AppendVector(A=@n22, B=@n23)
   )) {
   group=Character
   parameterName=Texture
   texture=/OfferCatalog/Textures/BattleRoyale/FeaturedItems/Outfit/S28/T-AthenaSoldiers-Character-VoidRedemption-Rebel
  }
  LinearInterpolate(A=LinearInterpolate(A=@n14, B=LinearInterpolate(A=@n17, B=@n18, Alpha=@n19), Alpha=NamedRerouteUsage()), B=@n24.RGB, Alpha=@n24.RGB)
  MaterialFunctionCall!(Input Alpha=Constant() {r=1}, bOverrideCornerRadius, Corner Radius Override) {materialFunction=/FortUILibrary/Systems/ItemShopTile/Materials/MaterialFunctions/MF_UI_ItemShop_RoundedCorners}
  n25 = MaterialFunctionCall(UVs=NamedRerouteUsage(), Texture Scale=ScalarParameter() {defaultValue=1, group=Radial Gradient, parameterName=Radial Gradient Scale}) {materialFunction=&ScaleUVsByCenter}
  n26 = AppendVector(
   A=ScalarParameter() {group=Radial Gradient, parameterName=Radial G X}
   B=ScalarParameter() {defaultValue=1, group=Radial Gradient, parameterName=Radial G Y}
  )
  n27 = MaterialFunctionCall("Value (0-1)"=MaterialFunctionCall(UV=Add(A=@n25.UVs, B=@n26), GradientStart=Constant(), GradientEnd=Constant() {r=1}) {materialFunction=&MF_UI_RadialGradient}.Smooth) {materialFunction=&MF_UI_EaseCurves}
  NamedRerouteDeclaration(None=Multiply(A=VectorParameter() {group=Radial Gradient, parameterName=Color, defaultValue=<rgba 0.0185 0.0185 0.0185 1>}.RGB, B=@n27.EaseInSine)) {name=Radial Gradient, nodeColor=<rgba 1 0 0.7266 1>}
  NamedRerouteDeclaration(None=Multiply(A=@n27.EaseInSine, B=ScalarParameter() {defaultValue=0.5, group=Radial Gradient, parameterName=Opacity})) {name=Radial G Mask, nodeColor=<rgba 0 1 0.9531 1>}
  NamedRerouteDeclaration(None=MaterialFunctionCall() {materialFunction=&MF_UI_UVScaling}.Scaled to Fit) {name=UV, nodeColor=<rgba 1 0.8438 0 1>}
  n28 = NamedRerouteUsage()
  n29 = AppendVector(A=ScalarParameter() {group=Primary Shape, parameterName=P X Location}, B=ScalarParameter() {group=Primary Shape, parameterName=P Y Location})
  n30 = MaterialFunctionCall(
   UVs=MaterialFunctionCall(UVs=Add(A=@n28, B=@n29), "Rotation Angle (0-1)"=ScalarParameter() {group=Primary Shape, parameterName=P Rotation}) {materialFunction=&CustomRotator}.Rotated Values
   Texture Scale=AppendVector(
    A=ScalarParameter() {defaultValue=1, group=Primary Shape, parameterName=P Scale X}
    B=ScalarParameter() {defaultValue=1, group=Primary Shape, parameterName=P Scale Y}
   )
  ) {materialFunction=&ScaleUVsByCenter}
  n31 = Add(
   A=LinearInterpolate(Alpha=NamedRerouteUsage()) {constA=-0.02, constB=0.025}
   B=ScalarParameter() {defaultValue=0.085, group=Shapes, parameterName=Shape 01 Rotation}
  )
  n32 = Constant() {desc=Polygon Sides, r=3}
  NamedRerouteDeclaration(None=MaterialFunctionCall!(
    UVs=Rotator(Coordinate=@n30.UVs, Time=@n31) {speed=6.2832}
    Size
    Sides=NamedRerouteDeclaration(None=@n32) {name=Polygon Sides, nodeColor=<rgba 0 0 0 1>}
    InnerAngle
    OutlineThickness
    OutlineAlignment
    GlowMin
    GlowMax
   ) {materialFunction=/FortUILibrary/MaterialFunctions/Shapes/MF_UI_RegularPolygon}.Fill) {name=Primary Shape, nodeColor=<rgba 0 0 0 1>}
  n33 = NamedRerouteUsage()
  n34 = AppendVector(
   A=ScalarParameter() {group=Tertiary Shape, parameterName=T X Location}
   B=ScalarParameter() {group=Tertiary Shape, parameterName=T Y Location}
  )
  n35 = MaterialFunctionCall(
   UVs=MaterialFunctionCall(UVs=Add(A=@n33, B=@n34), "Rotation Angle (0-1)"=ScalarParameter() {group=Tertiary Shape, parameterName=T Rotation}) {materialFunction=&CustomRotator}.Rotated Values
   Texture Scale=AppendVector(
    A=ScalarParameter() {defaultValue=1, group=Tertiary Shape, parameterName=T Scale X}
    B=ScalarParameter() {defaultValue=1, group=Tertiary Shape, parameterName=T Scale Y}
   )
  ) {materialFunction=&ScaleUVsByCenter}
  n36 = Add(
   A=LinearInterpolate(Alpha=NamedRerouteUsage()) {constA=0.02, constB=-0.03}
   B=ScalarParameter() {defaultValue=0.085, group=Shapes, parameterName=Shape 02 Rotation}
  )
  NamedRerouteDeclaration(None=MaterialFunctionCall(UVs=Rotator(Coordinate=@n35.UVs, Time=@n36) {speed=6.2832}, Sides=NamedRerouteUsage()) {materialFunction=&MF_UI_RegularPolygon}.Fill) {name=Tertiary Shape, nodeColor=<rgba 0 0 0 1>}
  NamedRerouteDeclaration(None=VectorParameter() {group=Color, parameterName=Color Theme 01, defaultValue=<rgba 0.6038 0.9131 1 1>}.RGB) {name=Color Theme 1, nodeColor=<rgba 0 0 0 1>}
  NamedRerouteDeclaration(None=VectorParameter() {group=Color, parameterName=Color Theme 02, defaultValue=<rgba 0.2195 0.7084 1 1>}.RGB) {name=Color Theme 2, nodeColor=<rgba 0 0 0 1>}
  NamedRerouteDeclaration(None=VectorParameter() {group=Color, parameterName=Color Theme 03, defaultValue=<rgba 0.013 0.1529 0.4287 1>}.RGB) {name=Color Theme 3, nodeColor=<rgba 0 0 0 1>}
  NamedRerouteDeclaration(None=VectorParameter() {group=Color, parameterName=Color Theme 04, defaultValue=<rgba 0.0027 0.0369 0.1651 1>}.RGB) {name=Color Theme 4, nodeColor=<rgba 0 0 0 1>}
  n37 = Time()
  n38 = Sine(None=Multiply(A=@n37) {constB=0.15})
  NamedRerouteDeclaration(None=Add(A=Multiply(A=@n38) {constB=0.5})) {name=Sine Time, nodeColor=<rgba 0 0 0 1>}
  n39 = Cosine(None=Multiply(A=@n37) {constB=0.2})
  NamedRerouteDeclaration(None=Add(A=Multiply(A=@n39) {constB=0.5})) {name=Cosine Time, nodeColor=<rgba 0 0 0 1>}
  n40 = Max(A=NamedRerouteUsage(), B=NamedRerouteUsage())
  n41 = NamedRerouteUsage()
  NamedRerouteDeclaration(None=Max(A=Max(A=@n40, B=@n41), B=NamedRerouteUsage())) {name=alphas, nodeColor=<rgba 0 0.2969 1 1>}
  n42 = NamedRerouteUsage()
  n43 = AppendVector(
   A=ScalarParameter() {group=Chonky Midshape, parameterName=C X Location}
   B=ScalarParameter() {group=Chonky Midshape, parameterName=C Y Location}
  )
  n44 = MaterialFunctionCall(
   UVs=MaterialFunctionCall(UVs=Add(A=@n42, B=@n43), "Rotation Angle (0-1)"=ScalarParameter() {group=Chonky Midshape, parameterName=C Rotation}) {materialFunction=&CustomRotator}.Rotated Values
   Texture Scale=AppendVector(
    A=ScalarParameter() {defaultValue=1, group=Chonky Midshape, parameterName=C Scale X}
    B=ScalarParameter() {defaultValue=1, group=Chonky Midshape, parameterName=C Scale Y}
   )
  ) {materialFunction=&ScaleUVsByCenter}
  n45 = Constant2Vector() {g=0.5, r=1}
  n46 = Add(
   A=LinearInterpolate(Alpha=NamedRerouteUsage()) {constA=-0.01, constB=0.02}
   B=ScalarParameter() {defaultValue=0.085, group=Shapes, parameterName=Chonk Rotation}
  )
  n47 = Saturate(None=ComponentMask(None=Rotator(Coordinate=@n44.UVs, Time=@n46) {speed=6.2832}) {r=true})
  n48 = Add(A=Multiply(A=Subtract(A=@n44.UVs, B=@n45), B=LinearInterpolate(Alpha=@n47) {constA=0.5, constB=0.5}), B=@n45)
  n49 = Constant() {r=4}
  n50 = MaterialFunctionCall(UVs=NamedRerouteUsage(), GradientStart=Constant(), GradientEnd=Constant() {r=1}, RotationAngle=Constant() {r=-0.85}) {materialFunction=&MF_UI_LinearGradient}
  NamedRerouteDeclaration(None=Min(
    A=MaterialFunctionCall(UVs=@n48, Sides=@n49) {materialFunction=&MF_UI_RegularPolygon}.Fill
    B=MaterialFunctionCall("Value (0-1)"=@n50.Result) {materialFunction=&MF_UI_EaseCurves}.EaseInSine
   )) {name=Chonky BG Triangle, nodeColor=<rgba 0 0 0 1>}
  NamedRerouteDeclaration(None=MaterialFunctionCall(UVs=@n6, Sides=NamedRerouteUsage()) {materialFunction=&MF_UI_RegularPolygon}.Fill) {name=Secondary Shape, nodeColor=<rgba 0 0 0 1>}
  NamedRerouteUsage()
 ]
}
