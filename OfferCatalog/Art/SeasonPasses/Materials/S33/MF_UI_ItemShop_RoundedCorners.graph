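# Material function for Item Shop UI in season pass art that trims an input alpha to rounded corners.
# Corner radius comes from MPC_ItemShop or an optional per-corner override, feeds MF_UI_CoreUI_RoundedMask in pixels, and the mask is combined with the input alpha via Min.
# A screen-position list-clipping branch builds anchored UV and height-relative radius reroutes, but none of them reach the output.
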MaterialFunction /OfferCatalog/Art/SeasonPasses/Materials/S33/MF_UI_ItemShop_RoundedCorners
graph {nodes [
  n1 = CollectionParameter() {collection=/FortUILibrary/Systems/ItemShopTile/MPC_ItemShop, parameterName=CornerRadius}
  n2 = LinearInterpolate(
   A=MaterialFunctionCall(R=@n1, G=@n1, B=@n1, A=@n1) {materialFunction=&AppendMany}.RGBA
   B=FunctionInput(Preview=Constant4Vector().RGBA) {bUsePreviewValueAsDefault=true, inputName=Corner Radius Override, inputType=FunctionInput_Vector4, sortPriority=2}
   Alpha=FunctionInput(Preview=Constant()) {bUsePreviewValueAsDefault=true, inputName=bOverrideCornerRadius, inputType=FunctionInput_Scalar, sortPriority=1}
  )
  n3 = Constant() {r=1}
  FunctionOutput(None=Min(
    A=FunctionInput() {inputName=Input Alpha, inputType=FunctionInput_Scalar}
    B=MaterialFunctionCall!(UV, Size, Corner Radius=@n2, Radius in Pixels=@n3, Skew Angle, Outline Thickness) {materialFunction=&MF_UI_CoreUI_RoundedMask}.Shape Mask
   )) {outputName=Final Alpha}
  n4 = Constant() {r=190}
  n5 = Subtract(
   A=ComponentMask(None=ScreenPosition().ViewportUV) {g=true}
   B=MaterialFunctionCall!(
    Value PX=NamedRerouteDeclaration(None=@n4) {name=List Clipping Limit, nodeColor=<rgba 0 0 0 1>}
    UV Pixel Size=ViewSize()
    X or Y=Constant() {r=1}
   ) {materialFunction=/FortUILibrary/MaterialFunctions/Basics/MF_UI_PixelsToNormalized}."Normalized (>1px)"
  )
  n6 = LinearInterpolate(
   A=MaterialFunctionCall(Value PX=Constant() {desc=Visible Area Height, r=59}, X or Y=Constant() {r=1}) {materialFunction=&MF_UI_PixelsToNormalized}."Normalized (>1px)"
   Alpha=Ceil(None=Saturate(None=@n5)) {desc=List View Clipping}
  ) {constA=0.25}
  NamedRerouteDeclaration(None=Divide(A=@n2, B=@n6)) {name=Corner Radius Relative to Height, nodeColor=<rgba 0 0 0 1>}
  n7 = NamedRerouteDeclaration(None=LinearInterpolate(A=Constant2Vector() {g=1, r=1}, B=Constant2Vector() {g=-1, r=-1}, Alpha=Constant2Vector() {desc=Anchor, g=1, r=0.5}) {constA=1, constB=-1}) {name=Anchor Range, nodeColor=<rgba 0 0 0 1>}
  n8 = Constant() {r=1}
  n9 = LinearInterpolate(
   A=Constant2Vector() {g=0.5, r=0.5}
   B=Constant2Vector()
   Alpha=NamedRerouteDeclaration(None=AppendVector(A=@n8, B=@n6)) {name=Screenspace Size, nodeColor=<rgba 0 0 0 1>}
  )
  NamedRerouteDeclaration(None=Add(A=MaterialFunctionCall() {materialFunction=&GetUserInterfaceUV}.Normalized UV, B=Multiply(A=@n7, B=@n9))) {name=UV, nodeColor=<rgba 0 0 0 1>}
 ]}
