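# Material function that draws the animated blue-purple Season 33 (JadeTowel) pass tile background for the UI, output as Focal Color.
# Gradient world and focal colors are built around the VaultWorld focal point function, then halftone and line layers are blended on top.
# Halftone dots come from a particle-grid circle SDF; lines come from rotated box SDFs run through ping-pong and wave functions.
# Named reroutes organize the masks, cosine-gradient tints, panned-texture glitches and sine-eased timing.
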MaterialFunction /OfferCatalog/Art/SeasonPasses/Materials/S33/MF_UI_JadeTowel_PassTileBG
graph {nodes [
  Constant() {r=8}
  n1 = MaterialFunctionCall!(
   Vault World Color=NamedRerouteUsage()
   Focal Color Override=NamedRerouteUsage()
   Focal Scale Override=LinearInterpolate(Alpha=NamedRerouteUsage()) {constA=1, constB=1.3}
   Mask Blend
  ) {materialFunction=/Game/Environments/FrontEnd/Blueprints/ItemPreview/VaultWorldTransitions/MaterialFunctions/MF_UI_VaultWorld_FocalPoint}
  n2 = LinearInterpolate(A=Add(A=NamedRerouteUsage(), B=@n1.Focal Color), B=NamedRerouteUsage(), Alpha=Multiply(A=NamedRerouteUsage()) {constB=0.2}) {constAlpha=0, desc=Half Tone}
  n3 = NamedRerouteUsage()
  n4 = Multiply(A=NamedRerouteUsage()) {constB=0.5}
  FunctionOutput(None=Saturate(None=LinearInterpolate(A=@n2, B=@n3, Alpha=@n4) {constAlpha=0, desc=Lines})) {outputName=Focal Color}
  MaterialFunctionCall() {materialFunction=&LinearGradient}
  MaterialFunctionCall() {materialFunction=&Blend_Overlay}
  n5 = NamedRerouteUsage()
  n6 = Add(A=@n5) {constB=-0.5}
  n7 = OneMinus(None=ComponentMask(None=Abs(None=@n6)) {r=true})
  Multiply(A=Saturate(None=Add(A=@n7) {constB=0.125}))
  n8 = Multiply(
   A=MaterialFunctionCall!("Use Local UV (Debug)", Texture Size, Scale Pivot Point) {materialFunction=/Game/Environments/FrontEnd/Blueprints/ItemPreview/VaultWorldTransitions/MaterialFunctions/MF_UI_VaultWorld_UV}.Scale X
   B=AppendVector(A=Constant() {r=8}, B=Constant() {r=8})
  ) {constB=6}
  n9 = NamedRerouteUsage()
  n10 = Constant() {r=12}
  Multiply(
   A=@n8
   B=LinearInterpolate(A=MaterialFunctionCall!(UV=@n9, Wave Height, Wave Length=@n10, Wave Speed=@n10, GlowMin, GlowMax, Outline Thickness, Custom Curve) {materialFunction=&/MF_UI_Wave}.SDF) {constAlpha=0.99}
  )
  n11 = MaterialFunctionCall(UVs=@n1.Focal Point UV, Texture Scale=ScalarParameter() {defaultValue=1.2, group=Focal Point, parameterName=FocalScale}) {materialFunction=&ScaleUVsByCenter}
  NamedRerouteDeclaration(None=@n11.UVs) {name=Focal_Point_UV, nodeColor=<rgba 1 0 0.0703 1>}
  n12 = Constant3Vector() {constant=<rgba 0.2051 0.147 0.7379 1>}
  n13 = SmoothStep(
   Min=LinearInterpolate(Alpha=NamedRerouteUsage()) {constB=0.2}
   Value=MaterialFunctionCall(UV Channel=NamedRerouteUsage()) {materialFunction=&LinearGradient}.VGradient
  )
  n14 = NamedRerouteUsage()
  n15 = NamedRerouteUsage()
  n16 = NamedRerouteUsage()
  n17 = LinearInterpolate(
   A=LinearInterpolate(
    A=Constant3Vector() {constant=<rgba 0 0.9495 1 1>}
    B=Constant3Vector() {constant=<rgba 0.0612 0.0437 0.552 1>}
    Alpha=MaterialFunctionCall(UVs=@n16) {materialFunction=&RadialGradientExponential}.RadialGradientExponential
   )
   B=Constant3Vector() {constant=<rgba 0.0409 0.2195 0.8308 1>}
   Alpha=SmoothStep(
    Min=LinearInterpolate(Alpha=@n15) {constA=0.4, constB=0.7}
    Value=MaterialFunctionCall(UV Channel=@n14) {materialFunction=&LinearGradient}.UGradient
   )
  )
  n18 = Power(Base=NamedRerouteUsage()) {constExponent=1.5}
  NamedRerouteDeclaration(None=LinearInterpolate(A=LinearInterpolate(A=@n17, B=@n12, Alpha=@n13), B=Constant3Vector() {constant=<rgba 0.1022 0.0018 0.5333 1>}, Alpha=OneMinus(None=@n18))) {name=Vault Worl Color, nodeColor=<rgba 1 0.3281 0 1>}
  n19 = Multiply(A=Time(), B=Multiply(A=Constant() {r=0.12}, B=ScalarParameter() {defaultValue=0.5, parameterName=AnimationSpeed}))
  NamedRerouteDeclaration(None=@n19) {name=CustomTime, nodeColor=<rgba 0.9219 0 1 1>}
  n20 = NamedRerouteUsage()
  n21 = LinearInterpolate(Alpha=@n20) {constA=2, constB=3.2}
  n22 = MaterialFunctionCall(UVs=NamedRerouteUsage()) {materialFunction=&RadialGradientExponential}
  n23 = MaterialFunctionCall!(
   Gradient=@n22.RadialGradientExponential
   a=MaterialFunctionCall(R=LinearInterpolate(Alpha=@n20), G=@n21, B=@n21) {materialFunction=&AppendMany}.RGB
   b
   c
   d
  ) {materialFunction=/FortUILibrary/MaterialFunctions/Color/MF_UI_CosineGradient}
  n24 = Saturate(None=@n23.Result)
  n25 = LinearInterpolate(A=Constant3Vector() {constant=<rgba 0 0.4809 1 1>}, B=Constant3Vector() {constant=<rgba 0 0.2536 1 1>}, Alpha=@n22.RadialGradientExponential)
  NamedRerouteDeclaration(None=Saturate(None=LinearInterpolate(A=@n25, B=@n24) {constAlpha=0.3})) {name=LinesColor, nodeColor=<rgba 1 0 0.4453 1>}
  n26 = Constant() {r=0.125}
  n27 = Constant() {r=0.25}
  n28 = Constant()
  n29 = TextureSample(UVs=Panner(
    Coordinate=MaterialFunctionCall(UVs=@n8, "Rotation Angle (0-1)"=@n26) {materialFunction=&CustomRotator}.Rotated Values
    Speed=AppendVector(A=@n27, B=@n28)
   )) {samplerType=SAMPLERTYPE_LinearColor, texture=&/TUIS7_BG}
  n30 = Constant() {r=0.15}
  n31 = MaterialFunctionCall(
   1=@n29.R
   2=@n29.R
   3=@n29.R
   A=Multiply(A=MaterialFunctionCall(Speed=@n30) {materialFunction=&DebugTimeSine}.Result)
  ) {materialFunction=&Lerp_Multiple_Float3}
  n32 = MaterialFunctionCall(UV Channel=MaterialFunctionCall(UVs=NamedRerouteUsage(), "Rotation Angle (0-1)"=Constant() {r=0.125}) {materialFunction=&CustomRotator}.Rotated Values) {materialFunction=&LinearGradient}
  n33 = SmoothStep(Value=@n32.VGradient) {constMax=1.2, constMin=0.9}
  n34 = SmoothStep(Value=OneMinus(None=@n32.VGradient)) {constMin=0.8}
  n35 = LinearInterpolate(B=Saturate(None=SmoothStep(Value=@n31.Lerp 3 Inputs) {constMax=0.5, constMin=0.3}), Alpha=Saturate(None=Max(A=@n33, B=@n34)))
  n36 = NamedRerouteUsage()
  NamedRerouteDeclaration(None=Min(A=LinearInterpolate(B=@n35) {constAlpha=0.25}, B=OneMinus(None=@n36))) {name=BG_Glitches, nodeColor=<rgba 0.3828 1 0 1>}
  NamedRerouteDeclaration(None=@n1.Focal Color) {name=FocalColor, nodeColor=<rgba 1 0 0.2344 1>}
  n37 = NamedRerouteUsage()
  n38 = LinearInterpolate(
   B=LinearInterpolate(A=Constant3Vector() {constant=<rgba 0.0255 0.4789 0.5156 1>}, B=Constant3Vector() {constant=<rgba 0 0.065 0.375 1>})
   Alpha=@n37
  )
  n39 = NamedRerouteUsage()
  NamedRerouteDeclaration(None=Min(
    A=MaterialFunctionCall(Base=@n38, Blend=@n39) {materialFunction=&Blend_Overlay}.Result
    B=@n37
   )) {name=GlitchColor, nodeColor=<rgba 0 0.5547 1 1>}
  NamedRerouteDeclaration(None=MaterialFunctionCall() {materialFunction=&MF_UI_UVScaling}.Scaled To Fill) {name=WorldUV-UV, nodeColor=<rgba 0 0.5547 1 1>}
  n40 = NamedRerouteUsage()
  n41 = Constant() {r=0.125}
  n42 = NamedRerouteUsage()
  n43 = MaterialFunctionCall!(
   Value=MaterialFunctionCall!(
    UVs=MaterialFunctionCall(UVs=@n40, "Rotation Angle (0-1)"=@n41) {materialFunction=&CustomRotator}.Rotated Values
    Size=LinearInterpolate(Alpha=@n42) {constA=0.95}
    CornerRoundness
    OutlineThickness
    OutlineAlignment
    GlowMin
    GlowMax
    Keep Aspect Ratio
   ) {materialFunction=&MF_UI_Box}.SDFOutline
   Scale=LinearInterpolate(Alpha=@n42) {constA=11, constB=12}
   Height
  ) {materialFunction=/FortUILibrary/MaterialFunctions/Basics/MF_UI_PingPong}
  n44 = MaterialFunctionCall(UV=@n43.Result, Wave Length=Constant() {r=0.2}, Wave Speed=Constant() {r=0.2}) {materialFunction=&MF_UI_Wave}
  n45 = LinearInterpolate(B=OneMinus(None=SmoothStep(Value=@n44.SDF Outline) {constMax=0.1, constMin=0.05})) {constAlpha=0.8}
  n46 = Min(
   A=Min(A=NamedRerouteDeclaration(None=@n45) {name=LinesEffectMask, nodeColor=<rgba 0.2891 0 1 1>}, B=NamedRerouteUsage())
   B=NamedRerouteUsage()
  )
  NamedRerouteDeclaration(None=Saturate(None=LinearInterpolate(B=@n46) {constAlpha=0.1})) {name=Line EffectGlitches, nodeColor=<rgba 1 0.6094 0 1>}
  n47 = NamedRerouteUsage()
  NamedRerouteDeclaration(None=LinearInterpolate(
    B=MaterialFunctionCall(Base=@n23.Result, Blend=@n47) {materialFunction=&Blend_Overlay}.Result
    Alpha=NamedRerouteUsage()
   ) {constAlpha=0.8}) {name=LineGlitchColor, nodeColor=<rgba 0.7578 0 1 1>}
  n48 = MaterialFunctionCall(UV Channel=NamedRerouteUsage()) {materialFunction=&LinearGradient}
  n49 = LinearInterpolate(B=@n43.Result, Alpha=OneMinus(None=SmoothStep(Value=@n48.VGradient) {constMin=0.4}))
  n50 = LinearInterpolate(Alpha=NamedRerouteUsage()) {constB=0.4}
  NamedRerouteDeclaration(None=Saturate(None=LinearInterpolate(B=@n49, Alpha=@n50) {constAlpha=0.4})) {name=LinesBGMask, nodeColor=<rgba 0 0.8359 1 1>}
  n51 = Constant3Vector() {constant=<rgba 0.202 0.2814 1 1>}
  n52 = MaterialFunctionCall(UV Channel=NamedRerouteUsage()) {materialFunction=&LinearGradient}
  n53 = Constant3Vector() {constant=<rgba 0 0.7383 1 1>}
  n54 = NamedRerouteUsage()
  NamedRerouteDeclaration(None=LinearInterpolate(
    A=Constant3Vector() {constant=<rgba 0 0.0519 1 1>}
    B=LinearInterpolate(A=@n51, B=@n53, Alpha=@n52.VGradient)
    Alpha=MaterialFunctionCall(UVs=@n54) {materialFunction=&RadialGradientExponential}.RadialGradientExponential
   )) {name=Focal Color Override, nodeColor=<rgba 0.0078 0 1 1>}
  NamedRerouteDeclaration(None=MaterialFunctionCall(UVs=@n1.Focal Point UV, Texture Scale=ScalarParameter() {defaultValue=1.2, group=Focal Point, parameterName=FocalMaskScale}) {materialFunction=&ScaleUVsByCenter}.UVs) {name=OGFocal_Point_UV, nodeColor=<rgba 0 1 0.6953 1>}
  NamedRerouteDeclaration(None=@n24) {name=ColorCosine, nodeColor=<rgba 1 0.5859 0 1>}
  n55 = ScalarParameter() {defaultValue=5, desc=speed, parameterName=Ripple Frequency}
  n56 = NamedRerouteUsage()
  n57 = OneMinus(None=Frac(None=Multiply(A=@n56) {constB=0.25}))
  n58 = Add(
   A=@n40
   B=AppendVector(
    A=ScalarParameter() {group=Focal Point, parameterName=WaveSquareX}
    B=ScalarParameter() {defaultValue=0.12, group=Focal Point, parameterName=WaveSquareY}
   )
  )
  n59 = Constant() {r=0.125}
  n60 = OneMinus(None=MaterialFunctionCall(
    UVs=MaterialFunctionCall(UVs=@n58, "Rotation Angle (0-1)"=@n59) {materialFunction=&CustomRotator}.Rotated Values
    Size=Constant() {r=1}
   ) {materialFunction=&MF_UI_Box}.SDF)
  NamedRerouteDeclaration(None=Saturate(None=MaterialFunctionCall!(Repetitions=@n55, Motion=@n57, SDF Input=@n60) {materialFunction=&/MF_UI_MPT_Wave}.SDF Output)) {name=FracWaveLine, nodeColor=<rgba 1 0.6797 0 1>}
  n61 = ComponentMask(None=OneMinus(None=@n5)) {g=true}
  n62 = Saturate(None=Multiply(A=Add(A=@n61) {constB=0.25}) {constB=1.5})
  n63 = Saturate(None=Add(A=ComponentMask(None=@n5) {g=true}) {constB=0.25})
  NamedRerouteDeclaration(None=Saturate(None=Multiply(A=@n62, B=@n63))) {name=Vignette, nodeColor=<rgba 0 1 0.4844 1>}
  NamedRerouteDeclaration(None=MaterialFunctionCall() {materialFunction=&MF_UI_UVScaling}.Scaled to Fit) {name=UVFill, nodeColor=<rgba 0.2422 0 1 1>}
  n64 = ScalarParameter() {defaultValue=1, group=Focal Point, parameterName=Halftone_FocalScale}
  NamedRerouteDeclaration(None=MaterialFunctionCall(
    UV=MaterialFunctionCall(UVs=@n1.Focal Point UV, Texture Scale=@n64) {materialFunction=&ScaleUVsByCenter}.UVs
    GradientStart=Constant()
    GradientEnd=Constant() {r=1}
   ) {materialFunction=&MF_UI_RadialGradient}.Result) {name=Halftone_FocalMask, nodeColor=<rgba 0.1484 1 0 1>}
  NamedRerouteDeclaration(None=MaterialFunctionCall(UV=@n11.UVs, GradientStart=Constant(), GradientEnd=Constant() {r=1}) {materialFunction=&MF_UI_RadialGradient}.Result) {name=Focal Mask, nodeColor=<rgba 0 0.1562 1 1>}
  n65 = NamedRerouteUsage()
  n66 = Constant() {r=0.125}
  n67 = ScalarParameter() {group=Focal Point, parameterName=SquareX}
  n68 = ScalarParameter() {group=Focal Point, parameterName=SquareY}
  n69 = MaterialFunctionCall(
   UVs=Add(
    A=MaterialFunctionCall(UVs=@n65, "Rotation Angle (0-1)"=@n66) {materialFunction=&CustomRotator}.Rotated Values
    B=AppendVector(A=@n67, B=@n68)
   )
   Size=Constant() {r=0.75}
   OutlineThickness=Constant() {r=0.05}
   GlowMax=Constant() {r=0.2}
  ) {materialFunction=&MF_UI_Box}
  NamedRerouteDeclaration(None=Saturate(None=@n69.Outline)) {name=LinesCenterMask, nodeColor=<rgba 1 0.6328 0 1>}
  n70 = Add(
   A=NamedRerouteUsage()
   B=AppendVector(
    A=ScalarParameter() {defaultValue=0.5, group=Focal Point, parameterName=Halftone X}
    B=ScalarParameter() {defaultValue=0.25, group=Focal Point, parameterName=Halftone Y}
   )
  )
  n71 = Constant() {r=1}
  n72 = NamedRerouteUsage()
  n73 = NamedRerouteUsage()
  n74 = Constant() {r=0.6}
  n75 = MaterialFunctionCall!(
   UV=MaterialFunctionCall!(
    UV=MaterialFunctionCall(UVs=@n70, Texture Scale=@n71) {materialFunction=&ScaleUVsByCenter}.UVs
    Global Offset
    Global Rotation
    Tiling=Constant() {r=25}
    Cell Offset=Constant()
    Cell Scale=LinearInterpolate(Alpha=@n72) {constA=2, constB=3.5}
    Cell Rotation
    Random Seed
   ) {materialFunction=/FortUILibrary/MaterialFunctions/Basics/MF_UI_ParticleGrid}.Cells UV
   Size=SmoothStep(Value=Multiply(A=@n73, B=@n74)) {constMin=0.04}
   OutlineThickness
   OutlineAlignment
   GlowMin
   GlowMax
  ) {materialFunction=&MF_UI_Circle}
  NamedRerouteDeclaration(None=Saturate(None=@n75.Fill)) {name=HalfTone, nodeColor=<rgba 0.6641 1 0 1>}
  n76 = NamedRerouteUsage()
  n77 = Min(A=OneMinus(None=SmoothStep(Value=@n76) {constMax=0.1}), B=NamedRerouteUsage())
  n78 = NamedRerouteUsage()
  NamedRerouteDeclaration(None=Saturate(None=Min(A=@n77, B=@n78))) {name=Lines Effect, nodeColor=<rgba 1 0.9141 0 1>}
  n79 = OneMinus(None=SmoothStep(Value=@n75.SDF) {constMax=0.4, constMin=0.3})
  NamedRerouteDeclaration(None=LinearInterpolate(B=LinearInterpolate(B=@n79) {constAlpha=0.3}, Alpha=@n73)) {name=HalfTone Glow, nodeColor=<rgba 0.2187 1 0 1>}
  n80 = NamedRerouteUsage()
  n81 = Constant() {r=0.35}
  NamedRerouteDeclaration(None=LinearInterpolate(
    A=Constant3Vector() {constant=<rgba 0.0782 0.2307 0.6654 1>}
    B=Constant3Vector() {constant=<rgba 0 0.4303 0.8385 1>}
    Alpha=MaterialFunctionCall(UVs=@n80, Radius=@n81) {materialFunction=&RadialGradientExponential}.RadialGradientExponential
   )) {name=HalFtone Color, nodeColor=<rgba 0 1 0.6484 1>}
  n82 = ConstantBiasScale(None=Sine(None=@n19))
  n83 = Subtract(A=MaterialFunctionCall("Value (0-1)"=@n82) {materialFunction=&MF_UI_EaseCurves}.EaseOutSine)
  n84 = LinearInterpolate(Alpha=@n82) {constA=-0.5, constB=-0.8}
  NamedRerouteDeclaration(None=Saturate(None=Multiply(A=@n83, B=@n84) {constB=-0.5})) {name=TimeAnim, nodeColor=<rgba 1 0 0.9141 1>}
  NamedRerouteUsage()
  NamedRerouteUsage()
  NamedRerouteUsage()
  OneMinus(None=@n69.Outline)
  Saturate(None=@n60)
  n85 = Add(B=Constant2Vector() {r=-0.1})
  Saturate(None=ComponentMask(None=Sine(None=@n85) {period=2}) {r=true})
 ]}
