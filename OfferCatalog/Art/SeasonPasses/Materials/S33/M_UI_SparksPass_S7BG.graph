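# Translucent UI material for a Sparks Pass Season 7 background, compositing a character texture over a shared pass tile backdrop.
# Character UVs are scaled to fill, then adjusted by center scale and X/Y offset parameters.
# The texture's RGB blends over the backdrop's Focal Color using itself as the per-channel weight, so dark areas show the backdrop.
# An item shop rounded-corners function masks a constant full alpha.
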Material /OfferCatalog/Art/SeasonPasses/Materials/S33/M_UI_SparksPass_S7BG
graph {
 properties {
  bAutomaticallySetUsageInEditor=false
  bUsedWithStaticMesh=false
  blendMode=BLEND_Translucent
  materialDomain=MD_UI
  shadingModel=MSM_Hair
  usedShadingModels=Hair
 }
 nodes [
  n1 = MaterialFunctionCall() {materialFunction=&MF_UI_UVScaling}
  n2 = ScalarParameter() {defaultValue=1, group=Character, parameterName=Texture Scale}
  n3 = ScalarParameter() {group=Character, parameterName=X Location}
  n4 = ScalarParameter() {group=Character, parameterName=Y Location}
  n5 = TextureSampleParameter2D(UVs=Add(
    A=MaterialFunctionCall(UVs=@n1.Scaled To Fill, Texture Scale=@n2) {materialFunction=&ScaleUVsByCenter}.UVs
    B=AppendVector(A=@n3, B=@n4)
   )) {
   group=Character
   parameterName=Texture
   texture=/OfferCatalog/Textures/BattleRoyale/FeaturedItems/Outfit/S28/T-AthenaSoldiers-Character-VoidRedemption-Rebel
  }
  LinearInterpolate(
   A=MaterialFunctionCall() {materialFunction=&/MF_UI_JadeTowel_PassTileBG}.Focal Color
   B=@n5.RGB
   Alpha=@n5.RGB
  )
  MaterialFunctionCall!(Input Alpha=Constant() {r=1}, bOverrideCornerRadius, Corner Radius Override) {materialFunction=&/MF_UI_ItemShop_RoundedCorners}
 ]
}
